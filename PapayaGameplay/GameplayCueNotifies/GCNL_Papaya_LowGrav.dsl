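; Looping gameplay cue notify that shows the Papaya low-gravity visual effect on a Fortnite player pawn.
; On loop start, it casts the target to a player pawn, spawns the NS_Papaya_LowGrav Niagara system attached to the pawn's mesh and stores the component.
; On removal, it deactivates the stored effect component.
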(event EventOnLoopingStart(Cascade)
  (params (MyTarget :type Actor%20Object%20Reference) (Parameters :type Gameplay%20Cue%20Parameters%20Structure%20%28by%20ref%29) (ParticleComponents :type Array%20of%20Cascade%20Particle%20System%20Component%20Object%20References) (AudioComponents :type Array%20of%20Audio%20Component%20Object%20References))
  (bind _asfort_player_pawn (Utilities|Casting|CastToFortPlayerPawn MyTarget))
  (bind _returnvalue (Effects|SpawnFXSystemAttached "/PapayaGameplay/Blueprints/Gadgets/FX/Niagara/NS_Papaya_LowGrav.NS_Papaya_LowGrav" (Variables|/Script/Engine.Character|GetMesh _asfort_player_pawn)))
  (Variables|Default|SetLow%20Gravity%20FX%20Asset _returnvalue))

(event EventOnRemoval(Cascade)
  (params (MyTarget :type Actor%20Object%20Reference) (Parameters :type Gameplay%20Cue%20Parameters%20Structure%20%28by%20ref%29) (ParticleComponents :type Array%20of%20Cascade%20Particle%20System%20Component%20Object%20References) (AudioComponents :type Array%20of%20Audio%20Component%20Object%20References) (BurstCameraShakeInstance :type Camera%20Shake%20Base%20Object%20Reference) (BurstDecalInstance :type Decal%20Actor%20Object%20Reference))
  (bind _low_gravity_fx_asset (Variables|Default|GetLow%20Gravity%20FX%20Asset)
    (:then
      (Components|Activation|Deactivate _low_gravity_fx_asset))
    (:else)))

(fn ConstructionScript)
