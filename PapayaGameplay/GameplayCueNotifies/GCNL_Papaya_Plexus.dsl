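; Looping gameplay cue that fades a Plexus mesh effect in and out for the locally controlled player pawn.
; On looping start, the local player's cue moves to a fixed world location and plays the PlexusFade timeline, while other clients' cues destroy their SM_MeshPlexus component.
; The timeline writes PlexusAlpha to the MPC_PapayaGadgetControls material parameter collection.
; On removal, the local cue stops and reverses that timeline to fade out.
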(event EventOnLoopingStart(Cascade)
  (params (MyTarget :type Actor%20Object%20Reference) (Parameters :type Gameplay%20Cue%20Parameters%20Structure%20%28by%20ref%29) (ParticleComponents :type Array%20of%20Cascade%20Particle%20System%20Component%20Object%20References) (AudioComponents :type Array%20of%20Audio%20Component%20Object%20References))
  (bind _asfort_player_pawn (Utilities|Casting|CastToFortPlayerPawn MyTarget))
  (if (Pawn|IsLocallyControlled _asfort_player_pawn)
    (Transformation|SetActorLocation "0,15000.000000, 0")
    (CallFunction|FadeinPlexus)
    (else
      (Components|DestroyComponent (Variables|Default|GetSM_MeshPlexus) 0))))

(event EventOnRemoval(Cascade)
  (params (MyTarget :type Actor%20Object%20Reference) (Parameters :type Gameplay%20Cue%20Parameters%20Structure%20%28by%20ref%29) (ParticleComponents :type Array%20of%20Cascade%20Particle%20System%20Component%20Object%20References) (AudioComponents :type Array%20of%20Audio%20Component%20Object%20References) (BurstCameraShakeInstance :type Camera%20Shake%20Base%20Object%20Reference) (BurstDecalInstance :type Decal%20Actor%20Object%20Reference))
  (bind _asfort_player_pawn (Utilities|Casting|CastToFortPlayerPawn MyTarget))
  (if (Pawn|IsLocallyControlled _asfort_player_pawn)
    (CallFunction|FadeOutPlexus)))

(event Custom|FadeIn%20Plexus
  (bind (_direction _plexuscontrol) (Timeline|PlexusFade PlayFromStart)
    (:Update
      (Rendering|Material|SetScalarParameterValue 0 "/Game/Environments/Papaya/Materials/MPC_PapayaGadgetControls.MPC_PapayaGadgetControls" "PlexusAlpha" _plexuscontrol :!class /Script/Engine.KismetMaterialLibrary))))

(event Custom|FadeOut%20Plexus
  (Utilities|FlowControl|Sequence
    (:then_0
      (Timeline|PlexusFade Stop))
    (:then_1
      (Timeline|PlexusFade Reverse))))

(fn ConstructionScript)
