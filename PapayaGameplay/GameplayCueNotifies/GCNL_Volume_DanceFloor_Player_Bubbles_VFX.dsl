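; Looping gameplay cue notify that shows a bubbles Niagara effect on a player, named for a dance floor volume.
; On loop start it casts the target to a Fort player pawn, caches its mesh, attaches NS_Player_Bubbles at the root socket and stores that component.
; On removal it deactivates the stored effect, then issues a DestroyComponent call using the cached mesh and the effect.
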(event EventOnLoopingStart(Cascade)
  (params (MyTarget :type Actor%20Object%20Reference) (Parameters :type Gameplay%20Cue%20Parameters%20Structure%20%28by%20ref%29) (ParticleComponents :type Array%20of%20Cascade%20Particle%20System%20Component%20Object%20References) (AudioComponents :type Array%20of%20Audio%20Component%20Object%20References))
  (bind _asfort_player_pawn (Utilities|Casting|CastToFortPlayerPawn MyTarget))
  (bind _output_get (Variables|Default|SetPlayerPawnSkelmesh (Variables|/Script/Engine.Character|GetMesh _asfort_player_pawn)))
  (bind _returnvalue (Effects|SpawnFXSystemAttached "/PapayaGameplay/Blueprints/Gadgets/FX/Niagara/PlayerLocal/NS_Player_Bubbles.NS_Player_Bubbles" _output_get "root"))
  (Variables|Default|SetParticleSystem _returnvalue))

(event EventOnRemoval(Cascade)
  (params (MyTarget :type Actor%20Object%20Reference) (Parameters :type Gameplay%20Cue%20Parameters%20Structure%20%28by%20ref%29) (ParticleComponents :type Array%20of%20Cascade%20Particle%20System%20Component%20Object%20References) (AudioComponents :type Array%20of%20Audio%20Component%20Object%20References) (BurstCameraShakeInstance :type Camera%20Shake%20Base%20Object%20Reference) (BurstDecalInstance :type Decal%20Actor%20Object%20Reference))
  (bind _particlesystem (Variables|Default|GetParticleSystem))
  (Components|Activation|Deactivate _particlesystem)
  (Components|DestroyComponent (Variables|Default|GetPlayerPawnSkelmesh) _particlesystem))

(fn ConstructionScript)
